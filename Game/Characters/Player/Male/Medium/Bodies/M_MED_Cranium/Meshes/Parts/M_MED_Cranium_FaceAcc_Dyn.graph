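# Physics asset for the dynamic hood bones of the medium male player Cranium body's face accessory mesh.
# Back, left, right and tip hood bones each have a capsule body, paired with a smaller capsule on a matching MSTR bone.
# Four constraints join each bone only to its MSTR partner, with tight twist limits and wider swing on the sides than on the back and tip.
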PhysicsAsset /Game/Characters/Player/Male/Medium/Bodies/M_MED_Cranium/Meshes/Parts/M_MED_Cranium_FaceAcc_Dyn
bodies {
 dyn_hood_bk [{shapeType=Capsule, length=2, radius=2, center=<1 0 0>, rotation=<rot 90 0 0>}]
 dyn_hood_bk_MSTR [{shapeType=Capsule, length=0.505, radius=0.505, rotation=<rot 0 -90 0>}]
 dyn_hood_l [{shapeType=Capsule, length=2, radius=2, center=<1 0 0>, rotation=<rot 90 0 0>}]
 dyn_hood_l_MSTR [{shapeType=Capsule, length=0.505, radius=0.505, rotation=<rot 0 -90 0>}]
 dyn_hood_r [{shapeType=Capsule, length=2, radius=2, center=<1 0 0>, rotation=<rot 90 0 0>}]
 dyn_hood_r_MSTR [{shapeType=Capsule, length=0.505, radius=0.505, rotation=<rot 0 -90 0>}]
 dyn_hood_tip [{shapeType=Capsule, length=2, radius=2, center=<1 0 0>, rotation=<rot 90 0 0>}]
 dyn_hood_tip_MSTR [{shapeType=Capsule, length=0.505, radius=0.505, rotation=<rot 0 -90 0>}]
}
constraints [
 {bone1Name=dyn_hood_r, bone2Name=dyn_hood_r_MSTR, swing1LimitDegrees=20, swing2LimitDegrees=45, twistLimitDegrees=5}
 {bone1Name=dyn_hood_bk, bone2Name=dyn_hood_bk_MSTR, swing1LimitDegrees=10, swing2LimitDegrees=20, twistLimitDegrees=5}
 {bone1Name=dyn_hood_l, bone2Name=dyn_hood_l_MSTR, swing1LimitDegrees=20, swing2LimitDegrees=45, twistLimitDegrees=5}
 {bone1Name=dyn_hood_tip, bone2Name=dyn_hood_tip_MSTR, swing1LimitDegrees=10, swing2LimitDegrees=20, twistLimitDegrees=5}
]
