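# Material function that adds an animated, glowing effect to Cranium player skins by overriding emissive, base color and specular where a mask input selects.
# Two panning triplanar noise samples, projected from pre-skinned position and normal, chain so the first warps the second.
# The resulting noise distorts detail texture UVs and breaks up a Fresnel term, both colored through a curve atlas.
# Fresnel also scales the dark base color, while specular is reduced.
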MaterialFunction /Game/Characters/Player/Male/Medium/Bodies/M_MED_Cranium/Skins/FX/Materials/MF_Cranium_FX
graph {nodes [
  n1 = Abs(None=VertexInterpolator(VS=PreSkinnedNormal()).PS)
  n2 = ComponentMask(None=@n1) {g=true}
  Abs(None=@n2)
  n3 = Abs(None=VertexNormalWS())
  ComponentMask(None=@n3) {b=true}
  ComponentMask(None=@n3) {r=true}
  Constant3Vector() {constant=<rgba 1 0 0 0>}
  n4 = VectorParameter() {group=DistortionFresnel, parameterName=DF_EmissiveFresnelNoiseTiling, defaultValue=<rgba 256 256 256 1>}
  n5 = FunctionInput(Preview=VertexInterpolator(VS=PreSkinnedPosition()).PS) {bUsePreviewValueAsDefault=true, inputName=PreskinnedPosition}
  n6 = VectorParameter() {group=DistortionFresnel, parameterName=DF_DistortionPanSpeedAndDirection, defaultValue=<rgba -0.25 -0.25 0 1>}
  n7 = Add(
   A=Divide(A=@n5, B=VectorParameter() {group=DistortionFresnel, parameterName=DF_EmissiveDistortionTiling, defaultValue=<rgba 256 256 256 1>}.RGB)
   B=Multiply(A=Time(), B=@n6.RGB)
  )
  n8 = TextureObjectParameter() {parameterName=DistorntionNoise, texture=/Game/Global/Textures/Default/Noise/3xPerlin_M}
  n9 = LinearInterpolate(
   A=TextureSample(UVs=ComponentMask(None=@n7) {g=true, r=true}, Tex=@n8).RGB
   B=TextureSample(UVs=ComponentMask(None=@n7) {b=true, r=true}, Tex=@n8).RGB
   Alpha=@n2
  )
  n10 = TextureSample(UVs=ComponentMask(None=@n7) {b=true, g=true}, Tex=@n8)
  n11 = ComponentMask(None=@n1) {r=true}
  n12 = Subtract(A=Multiply(A=LinearInterpolate(A=@n9, B=@n10.RGB, Alpha=@n11)) {constB=2})
  n13 = VectorParameter() {group=DistortionFresnel, parameterName=DF_DistortionStrength, defaultValue=<rgba 0.2 0.2 0.2 1>}
  n14 = VectorParameter() {group=DistortionFresnel, parameterName=DF_DistortionNoisePanSpeedAndDirection, defaultValue=<rgba -0.1 -0.5 0 1>}
  n15 = Add(A=Add(A=Divide(A=@n5, B=@n4.RGB), B=Multiply(A=@n12, B=@n13.RGB)), B=Multiply(A=Time(), B=@n14.RGB))
  n16 = ComponentMask(None=@n15) {g=true, r=true}
  n17 = TextureObjectParameter() {parameterName=DistorntionNoise, texture=&3xPerlin_M}
  n18 = ComponentMask(None=@n15) {b=true, r=true}
  n19 = LinearInterpolate(
   A=LinearInterpolate(A=TextureSample(UVs=@n16, Tex=@n17).RGB, B=TextureSample(UVs=@n18, Tex=@n17).RGB, Alpha=@n2)
   B=TextureSample(UVs=ComponentMask(None=@n15) {b=true, g=true}, Tex=@n17).RGB
   Alpha=@n11
  )
  n20 = Subtract(A=Multiply(A=ComponentMask(None=@n19) {g=true, r=true}) {constB=2})
  n21 = FunctionInput(Preview=Constant4Vector().RGBA) {bUsePreviewValueAsDefault=true, inputName=4, inputType=FunctionInput_Vector4}
  n22 = Multiply(
   A=ScalarParameter() {defaultValue=0.005, parameterName=NoiseDistortion_Strength}
   B=OneMinus(None=ChannelMaskParameter(None=@n21) {maskChannel=Blue, parameterName=FX_MaskChannelB, defaultValue=<rgba 0 0 1 0>}.R)
  ) {constA=1}
  n23 = TextureSampleParameter2D(UVs=Add(A=TextureCoordinate(), B=Multiply(A=@n20, B=@n22))) {
   parameterName=Detailtexture
   texture=/Game/Characters/Player/Male/Medium/Bodies/M_MED_Cranium/Skins/FX/Textures/T_M_MED_Cranium_Green2_Body_E
  }
  n24 = CurveAtlasRowParameter(CurveTime=@n23.G) {
   atlas=&AthenaCurveAtlas01
   curve=/Game/Characters/Player/Male/Medium/Bodies/M_MED_Cranium/Skins/FX/Textures/CraniumColorCurve
   defaultValue=284
   parameterName=Cranium_ColorCurve
  }
  n25 = GetMaterialAttributes(None=FunctionInput() {inputName=MaterialAttributes, inputType=FunctionInput_MaterialAttributes}) {attributeGetTypes [#1, #2]}
  n26 = Add(A=@n25.EmissiveColor, B=Multiply(A=@n24, B=ScalarParameter() {defaultValue=1, parameterName=Cranium_EmissiveStrength}))
  n27 = Power(
   Base=Fresnel(ExponentIn=ScalarParameter() {defaultValue=1, parameterName=FresnelExponent}, Normal=PixelNormalWS())
   Exp=ScalarParameter() {defaultValue=1.85, parameterName=FresnelPower}
  )
  n28 = Clamp(None=@n27) {minDefault=0.05}
  n29 = ComponentMask(None=@n19) {r=true}
  n30 = ScalarParameter() {defaultValue=0.25, parameterName=DistortionStrength}
  n31 = Clamp(None=Subtract(A=@n27, B=Multiply(A=@n29, B=@n30)))
  n32 = CurveAtlasRowParameter(CurveTime=@n31) {
   atlas=&AthenaCurveAtlas01
   curve=&CraniumColorCurve
   defaultValue=284
   parameterName=Cranium_ColorCurve
  }
  n33 = ScalarParameter() {defaultValue=1, parameterName=ColorCurveGlowStrength}
  n34 = Multiply(
   A=Add(A=Multiply(A=@n26, B=@n28), B=Multiply(A=@n32, B=@n33))
   B=VectorParameter() {parameterName=Cranium_EmissiveTint, defaultValue=<rgba 1 1 1 1>}.RGB
  )
  n35 = Add(A=Multiply(A=VectorParameter() {parameterName=DiffuseOverrideColor, defaultValue=<rgba 0.02 0.02 0.02 1>}.RGB, B=@n27)) {constB=0.005}
  n36 = ChannelMaskParameter(None=@n21) {parameterName=FX_Maskhannel}
  n37 = LinearInterpolate(A=@n25.Specular, B=Multiply(A=@n25.Specular) {constB=0.15}, Alpha=@n36.R)
  FunctionOutput(None=BlendMaterialAttributes(
    A=@n25.MaterialAttributes
    B=SetMaterialAttributes!(MaterialAttributes=@n25.MaterialAttributes, EmissiveColor=@n34, BaseColor=@n35, Specular=@n37) {attributeSetTypes [#1, #3, #2]}
    Alpha=@n36.R
   ))
  Multiply(A=@n31, B=VectorParameter() {parameterName=Cranium_EmissiveColor, defaultValue=<rgba 0.3763 0.7454 0 1>}.RGB)
  Multiply(A=@n23.G)
  TextureSampleParameter2D() {parameterName=SkinFX_Mask, texture=&T_M_MED_Cranium_Green2_Body_E}
 ]}
