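# Master character material for the male medium Cranium player skin, supporting Default Lit, Subsurface Profile and Cloth shading.
# Shared PreFX base output passes through static-switch layers for cloth fuzz, silk, distortion fresnel and metal LUT, driven by one SkinFX_Mask texture.
# MF_Cranium_FX then applies skin-specific effects using a second mask and pre-skinned local position, before the shared PostFX function.
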Material /Game/Characters/Player/Male/Medium/Bodies/M_MED_Cranium/Skins/FX/Materials/M_MED_Cranium_MAster.M_MED_Cranium_Master
graph {
 properties {
  bUseMaterialAttributes=true
  bUsedWithMeshDeformer=true
  bUsedWithMorphTargets=true
  bUsedWithSkeletalMesh=true
  pixelDepthOffsetMode=PDOM_Legacy
  shadingModel=MSM_FromMaterialExpression
  usedShadingModels=Default Lit | Subsurface Profile | Cloth
 }
 nodes [
  n1 = MaterialFunctionCall() {materialFunction=&BaseFN_CharacterFunction_PreFX}
  n2 = TextureSampleParameter2D() {
   parameterName=Masl_02
   samplerType=SAMPLERTYPE_Masks
   texture=/Game/Characters/Player/Male/Medium/Bodies/M_MED_Cranium/Skins/FX/Textures/T_M_MED_Cranium_Green2_Accessory_FX_02
  }
  n3 = TextureSample(Tex=TextureObjectParameter() {group=Skin FX, parameterName=SkinFX_Mask})
  n4 = StaticSwitchParameter(
   True=MaterialFunctionCall!(
    Alpha=StaticComponentMaskParameter(None=@n3.RGBA) {defaultR=true, group=Skin FX, parameterName=Cloth Channel}
    Gmap=@n1.Gmap
    Material=@n1.Gmap
   ) {materialFunction=/Game/Weapons/WeaponSkins/MaterialFunction/ClothFuzz}.Result
   False=@n1.Result
  ) {group=Skin FX, parameterName=Use Cloth fuzz}
  n5 = StaticSwitchParameter(
   True=MaterialFunctionCall!(Alpha=ChannelMaskParameter(None=@n3.RGBA) {group=Skin FX, parameterName=Silk Channel}.R, Material=@n4) {materialFunction=/Game/Weapons/WeaponSkins/MaterialFunction/MF_Silk}.Result
   False=@n4
  ) {group=Skin FX, parameterName=Use Silk}
  n6 = StaticSwitchParameter(
   True=MaterialFunctionCall!(Mask=@n3.RGBA, Material=@n5) {materialFunction=/Game/Weapons/WeaponSkins/MaterialFunction/MF_DistortionFresnel}.Result
   False=@n5
  ) {group=Skin FX, parameterName=UseDistortionFresnel}
  n7 = MaterialFunctionCall(Mask=@n3.RGBA, Materials=@n6) {materialFunction=&MF_MetalsLUT}
  MaterialFunctionCall(
   AO=@n1.AO
   Distance Blend=@n1.AO
   Fresnel=@n1.AO
   Gmap=@n1.AO
   Pre Skinned Local Position=@n1.AO
   Material=MaterialFunctionCall!(
    4=@n2.RGBA
    PreskinnedPosition=@n1.Pre skinned Local Position
    MaterialAttributes=StaticSwitchParameter(True=@n7.Result, False=@n6) {group=Skin FX, parameterName=UseMetalLUT}
   ) {materialFunction=&/MF_Cranium_FX}.Result
  ) {materialFunction=&BaseFN_CharacterFunction_PostFX}
 ]
}
